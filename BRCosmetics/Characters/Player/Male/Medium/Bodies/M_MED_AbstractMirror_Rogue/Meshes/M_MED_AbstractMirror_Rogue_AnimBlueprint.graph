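# Anim Blueprint for the M_MED_AbstractMirror_Rogue player body, adding secondary motion to its gun holster and pouches.
# It copies the attached parent's pose through a Control Rig, then weighted bone constraints position the master attachment joints before caching StartPose.
# Chained Anim Dynamics springs jiggle the dyn joints from that pose, with holster simulation reduced to quarter alpha when bIsInFrontEnd is true.
# The EventGraph holds no logic.
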AnimBlueprint /BRCosmetics/Characters/Player/Male/Medium/Bodies/M_MED_AbstractMirror_Rogue/Meshes/M_MED_AbstractMirror_Rogue_AnimBlueprint
target_skeleton=&/M_MED_AbstractMirror_Rogue_Skeleton
graphs [
 AnimGraph {nodes [
   n1 = Anim Dynamics - Bone: dyn_gun_holster_jnt(ComponentPose=Local To Component(LocalPose=Use cached pose 'StartPose'()), Alpha=1, ExternalForce="0,0,0") {
    angularDampingOverride=0.99
    angularSpringConstant=20000
    bAngularSpring=true
    bOverrideAngularDamping=true
    bUseGravityOverride=true
    simulationSpace=BoneRelative
    alphaBoolBlend {blendOption=Linear}
    boundBone {boneName=dyn_gun_holster_jnt}
    componentAppliedLinearAccClamp=<3000 3000 3000>
    componentLinearAccScale=<0.3 0.3 0.5>
    componentLinearVelScale=<0.3 0.3 0.5>
    gravityOverride {z=-980}
    physicsBodyDefinitions [{
      boundBone {boneName=dyn_gun_holster_jnt}
      boxExtents=<35 35 10>
      constraintSetup {angularLimitsMax=<10 10 35>, angularLimitsMin=<-10 5 -35>, angularTarget=<0 0 0>}
      localJointOffset=<0 0 -2>
     }]
    relativeSpaceBone {boneName=thigh_r}
   }
   n2 = Component To Local(ComponentPose=@n1)
   n3 = Anim Dynamics - Bone: dyn_gun_holster_jnt(ComponentPose=Local To Component(LocalPose=Use cached pose 'StartPose'()), Alpha=0.25, ExternalForce="0,0,0") {
    angularDampingOverride=0.99
    angularSpringConstant=20000
    bAngularSpring=true
    bOverrideAngularDamping=true
    bUseGravityOverride=true
    simulationSpace=BoneRelative
    alphaBoolBlend {blendOption=Linear}
    boundBone {boneName=dyn_gun_holster_jnt}
    componentAppliedLinearAccClamp=<3000 3000 3000>
    componentLinearAccScale=<0.3 0.3 0.5>
    componentLinearVelScale=<0.3 0.3 0.5>
    gravityOverride {z=-980}
    physicsBodyDefinitions [{
      boundBone {boneName=dyn_gun_holster_jnt}
      boxExtents=<35 35 10>
      constraintSetup {angularLimitsMax=<10 10 35>, angularLimitsMin=<-10 5 -35>, angularTarget=<0 0 0>}
      localJointOffset=<0 0 -2>
     }]
    relativeSpaceBone {boneName=thigh_r}
   }
   n4 = Component To Local(ComponentPose=@n3)
   n5 = Get bIsInFrontEnd()
   n6 = Anim Dynamics - Bone: dyn_r_hip_pouch_jnt(
    ComponentPose=Local To Component(LocalPose=Blend Poses by bool(BlendPose_0=@n4, BlendPose_1=@n2, bActiveValue=@n5) {blendType=Linear})
    Alpha=1
    ExternalForce="0,0,0"
   ) {
    angularDampingOverride=0.99
    angularSpringConstant=20000
    bAngularSpring=true
    bOverrideAngularDamping=true
    bUseGravityOverride=true
    simulationSpace=BoneRelative
    alphaBoolBlend {blendOption=Linear}
    boundBone {boneName=dyn_r_hip_pouch_jnt}
    componentAppliedLinearAccClamp=<3000 3000 3000>
    componentLinearAccScale=<0.3 0.3 0.3>
    componentLinearVelScale=<0.3 0.3 0.3>
    gravityOverride {z=-980}
    physicsBodyDefinitions [{
      boundBone {boneName=dyn_r_hip_pouch_jnt}
      boxExtents=<15 10 15>
      constraintSetup {angularLimitsMax=<15 5 0>, angularLimitsMin=<-15 -5 0>, angularTarget=<0 0 0>}
      localJointOffset=<0 3 0>
     }]
    relativeSpaceBone {boneName=thigh_r}
   }
   n7 = Anim Dynamics - Bone: dyn_l_outer_chest_pouch_jnt(ComponentPose=@n6, Alpha=1, ExternalForce="0,0,0") {
    angularDampingOverride=0.99
    angularSpringConstant=12000
    bAngularSpring=true
    bOverrideAngularDamping=true
    bUseGravityOverride=true
    simulationSpace=RootRelative
    alphaBoolBlend {blendOption=Linear}
    boundBone {boneName=dyn_l_outer_chest_pouch_jnt}
    componentAppliedLinearAccClamp=<3000 3000 3000>
    componentLinearAccScale=<0.4 0.2 0.2>
    componentLinearVelScale=<0.4 0.2 0.2>
    gravityOverride {z=-980}
    physicsBodyDefinitions [{
      boundBone {boneName=dyn_l_outer_chest_pouch_jnt}
      boxExtents=<20 10 10>
      constraintSetup {angularLimitsMax=<5 2 2>, angularLimitsMin=<-5 -2 -2>, angularTarget=<0 0 0>}
      localJointOffset=<-5 0 0>
     }]
    relativeSpaceBone {boneName=spine_05}
   }
   n8 = Anim Dynamics - Bone: dyn_l_inner_chest_pouch_jnt(ComponentPose=@n7, Alpha=1, ExternalForce="0,0,0") {
    angularDampingOverride=0.99
    angularSpringConstant=12000
    bAngularSpring=true
    bOverrideAngularDamping=true
    bUseGravityOverride=true
    simulationSpace=RootRelative
    alphaBoolBlend {blendOption=Linear}
    boundBone {boneName=dyn_l_inner_chest_pouch_jnt}
    componentAppliedLinearAccClamp=<3000 3000 3000>
    componentLinearAccScale=<0.4 0.2 0.2>
    componentLinearVelScale=<0.4 0.2 0.2>
    gravityOverride {z=-980}
    physicsBodyDefinitions [{
      boundBone {boneName=dyn_l_inner_chest_pouch_jnt}
      boxExtents=<20 10 10>
      constraintSetup {angularLimitsMax=<5 2 2>, angularLimitsMin=<-5 -2 -2>, angularTarget=<0 0 0>}
      localJointOffset=<-5 0 0>
     }]
    relativeSpaceBone {boneName=spine_05}
   }
   n9 = Anim Dynamics - Bone: dyn_r_outer_chest_pouch_jnt(ComponentPose=@n8, Alpha=1, ExternalForce="0,0,0") {
    angularDampingOverride=0.99
    angularSpringConstant=12000
    bAngularSpring=true
    bOverrideAngularDamping=true
    bUseGravityOverride=true
    simulationSpace=RootRelative
    alphaBoolBlend {blendOption=Linear}
    boundBone {boneName=dyn_r_outer_chest_pouch_jnt}
    componentAppliedLinearAccClamp=<3000 3000 3000>
    componentLinearAccScale=<0.4 0.2 0.2>
    componentLinearVelScale=<0.4 0.2 0.2>
    gravityOverride {z=-980}
    physicsBodyDefinitions [{
      boundBone {boneName=dyn_r_outer_chest_pouch_jnt}
      boxExtents=<20 10 10>
      constraintSetup {angularLimitsMax=<5 2 2>, angularLimitsMin=<-5 -2 -2>, angularTarget=<0 0 0>}
      localJointOffset=<5 0 0>
     }]
    relativeSpaceBone {boneName=spine_05}
   }
   n10 = Anim Dynamics - Bone: dyn_r_inner_chest_pouch_jnt(ComponentPose=@n9, Alpha=1, ExternalForce="0,0,0") {
    angularDampingOverride=0.99
    angularSpringConstant=12000
    bAngularSpring=true
    bOverrideAngularDamping=true
    bUseGravityOverride=true
    simulationSpace=RootRelative
    alphaBoolBlend {blendOption=Linear}
    boundBone {boneName=dyn_r_inner_chest_pouch_jnt}
    componentAppliedLinearAccClamp=<3000 3000 3000>
    componentLinearAccScale=<0.4 0.2 0.2>
    componentLinearVelScale=<0.4 0.2 0.2>
    gravityOverride {z=-980}
    physicsBodyDefinitions [{
      boundBone {boneName=dyn_r_inner_chest_pouch_jnt}
      boxExtents=<20 10 10>
      constraintSetup {angularLimitsMax=<5 2 2>, angularLimitsMin=<-5 -2 -2>, angularTarget=<0 0 0>}
      localJointOffset=<5 0 0>
     }]
    relativeSpaceBone {boneName=spine_05}
   }
   n11 = Anim Dynamics - Bone: dyn_l_outer_hip_pouch(ComponentPose=@n10, Alpha=1, ExternalForce="0,0,0") {
    angularDampingOverride=1
    angularSpringConstant=12000
    bAngularSpring=true
    bOverrideAngularDamping=true
    bUseGravityOverride=true
    simulationSpace=RootRelative
    alphaBoolBlend {blendOption=Linear}
    boundBone {boneName=dyn_l_outer_hip_pouch}
    componentAppliedLinearAccClamp=<3000 3000 3000>
    componentLinearAccScale=<0.4 0.2 0.2>
    componentLinearVelScale=<0.4 0.2 0.2>
    gravityOverride {z=-980}
    physicsBodyDefinitions [{
      boundBone {boneName=dyn_l_outer_hip_pouch}
      boxExtents=<20 10 10>
      constraintSetup {angularLimitsMax=<5 5 2>, angularLimitsMin=<-5 -5 -2>, angularTarget=<0 0 0>}
      localJointOffset=<5 0 0>
     }]
    relativeSpaceBone {boneName=spine_05}
   }
   n12 = Anim Dynamics - Bone: dyn_l_middle_hip_pouch(ComponentPose=@n11, Alpha=1, ExternalForce="0,0,0") {
    angularDampingOverride=1
    angularSpringConstant=12000
    bAngularSpring=true
    bOverrideAngularDamping=true
    bUseGravityOverride=true
    simulationSpace=RootRelative
    alphaBoolBlend {blendOption=Linear}
    boundBone {boneName=dyn_l_middle_hip_pouch}
    componentAppliedLinearAccClamp=<3000 3000 3000>
    componentLinearAccScale=<0.4 0.2 0.2>
    componentLinearVelScale=<0.4 0.2 0.2>
    gravityOverride {z=-980}
    physicsBodyDefinitions [{
      boundBone {boneName=dyn_l_middle_hip_pouch}
      boxExtents=<20 10 10>
      constraintSetup {angularLimitsMax=<5 5 2>, angularLimitsMin=<-5 -5 -2>, angularTarget=<0 0 0>}
      localJointOffset=<5 0 0>
     }]
    relativeSpaceBone {boneName=spine_05}
   }
   n13 = Anim Dynamics - Bone: dyn_l_inner_hip_pouch(ComponentPose=@n12, Alpha=1, ExternalForce="0,0,0") {
    angularDampingOverride=1
    angularSpringConstant=12000
    bAngularSpring=true
    bOverrideAngularDamping=true
    bUseGravityOverride=true
    simulationSpace=RootRelative
    alphaBoolBlend {blendOption=Linear}
    boundBone {boneName=dyn_l_inner_hip_pouch}
    componentAppliedLinearAccClamp=<3000 3000 3000>
    componentLinearAccScale=<0.4 0.2 0.2>
    componentLinearVelScale=<0.4 0.2 0.2>
    gravityOverride {z=-980}
    physicsBodyDefinitions [{
      boundBone {boneName=dyn_l_inner_hip_pouch}
      boxExtents=<20 10 10>
      constraintSetup {angularLimitsMax=<5 5 2>, angularLimitsMin=<-5 -5 -2>, angularTarget=<0 0 0>}
      localJointOffset=<5 0 0>
     }]
    relativeSpaceBone {boneName=spine_05}
   }
   Output Pose(Result=Component To Local(ComponentPose=@n13)) {layerGroup=None}
   n14 = Control Rig(Source=Copy Pose From Mesh() {bUseAttachedParent=true}) {
    alphaBoolBlend {blendOption=Linear}
    controlRigAssetReference {blueprintRigClass=&/M_MED_AbstractMirror_Rogue_CtrlRig.M_MED_AbstractMirror_Rogue_CtrlRig_C}
   }
   n15 = "Constraint - master_l_outer_chest_pouch_jnt (PPPPPPP)"!(
    ComponentPose=Local To Component(LocalPose=@n14)
    Alpha=1
    ConstraintWeights_0=0.0359
    ConstraintWeights_1:float=0.0987
    ConstraintWeights_2:float=0.2455
    ConstraintWeights_3:float=0.4031
    ConstraintWeights_4:float=0.1411
    ConstraintWeights_5:float=0.0536
    ConstraintWeights_6:float=0.0221
   ) -> (Pose) {
    alphaBoolBlend {blendOption=Linear}
    boneToModify {boneName=master_l_outer_chest_pouch_jnt}
    constraintSetup [
     {targetBone {boneName=pelvis}}
     {targetBone {boneName=spine_01}}
     {targetBone {boneName=spine_02}}
     {targetBone {boneName=spine_03}}
     {targetBone {boneName=spine_04}}
     {targetBone {boneName=clavicle_l}}
     {targetBone {boneName=deform_pec_l}}
    ]
    constraintWeights [1,1,1,1,1,1,1]
   }
   n16 = "Constraint - master_l_inner_chest_pouch_jnt (PPPPPPP)"!(
    ComponentPose=@n15
    Alpha=1
    ConstraintWeights_0=0.0382
    ConstraintWeights_1:float=0.1264
    ConstraintWeights_2:float=0.2589
    ConstraintWeights_3:float=0.2118
    ConstraintWeights_4:float=0.0288
    ConstraintWeights_5:float=0.116
    ConstraintWeights_6:float=0.2198
   ) -> (Pose) {
    alphaBoolBlend {blendOption=Linear}
    boneToModify {boneName=master_l_inner_chest_pouch_jnt}
    constraintSetup [
     {targetBone {boneName=spine_01}}
     {targetBone {boneName=spine_02}}
     {targetBone {boneName=spine_03}}
     {targetBone {boneName=spine_04}}
     {targetBone {boneName=spine_05}}
     {targetBone {boneName=clavicle_l}}
     {targetBone {boneName=deform_pec_l}}
    ]
    constraintWeights [1,1,1,1,1,1,1]
   }
   n17 = "Constraint - master_r_outer_chest_pouch_jnt (PPPPPPPP)"!(
    ComponentPose=@n16
    Alpha=1
    ConstraintWeights_0=0.0495
    ConstraintWeights_1:float=0.041
    ConstraintWeights_2:float=0.1159
    ConstraintWeights_3:float=0.2345
    ConstraintWeights_4:float=0.1799
    ConstraintWeights_5:float=0.0618
    ConstraintWeights_6:float=0.1521
    ConstraintWeights_7:float=0.1653
   ) -> (Pose) {
    alphaBoolBlend {blendOption=Linear}
    boneToModify {boneName=master_r_outer_chest_pouch_jnt}
    constraintSetup [
     {targetBone {boneName=pelvis}}
     {targetBone {boneName=spine_01}}
     {targetBone {boneName=spine_02}}
     {targetBone {boneName=spine_03}}
     {targetBone {boneName=spine_04}}
     {targetBone {boneName=spine_05}}
     {targetBone {boneName=clavicle_r}}
     {targetBone {boneName=deform_pec_r}}
    ]
    constraintWeights [1,1,1,1,1,1,1,1]
   }
   n18 = "Constraint - master_r_inner_chest_pouch_jnt (PPPPPPPP)"!(
    ComponentPose=@n17
    Alpha=1
    ConstraintWeights_0=0.0731
    ConstraintWeights_1:float=0.0814
    ConstraintWeights_2:float=0.1955
    ConstraintWeights_3:float=0.331
    ConstraintWeights_4:float=0.1329
    ConstraintWeights_5:float=0.0423
    ConstraintWeights_6:float=0.1101
    ConstraintWeights_7:float=0.0337
   ) -> (Pose) {
    alphaBoolBlend {blendOption=Linear}
    boneToModify {boneName=master_r_inner_chest_pouch_jnt}
    constraintSetup [
     {targetBone {boneName=pelvis}}
     {targetBone {boneName=spine_01}}
     {targetBone {boneName=spine_02}}
     {targetBone {boneName=spine_03}}
     {targetBone {boneName=spine_04}}
     {targetBone {boneName=spine_05}}
     {targetBone {boneName=clavicle_r}}
     {targetBone {boneName=deform_pec_r}}
    ]
    constraintWeights [1,1,1,1,1,1,1,1]
   }
   n19 = "Constraint - master_r_hip_pouch_jnt (PPPPPPPP)"!(
    ComponentPose=@n18
    Alpha=1
    ConstraintWeights_0=0.3508
    ConstraintWeights_1:float=0.1382
    ConstraintWeights_2:float=0.1801
    ConstraintWeights_3:float=0.1511
    ConstraintWeights_4:float=0.0443
    ConstraintWeights_5:float=0.0423
    ConstraintWeights_6:float=0.0744
    ConstraintWeights_7:float=0.0189
   ) -> (Pose) {
    alphaBoolBlend {blendOption=Linear}
    boneToModify {boneName=master_r_hip_pouch_jnt}
    constraintSetup [
     {targetBone {boneName=pelvis}}
     {targetBone {boneName=spine_01}}
     {targetBone {boneName=spine_02}}
     {targetBone {boneName=spine_03}}
     {targetBone {boneName=spine_04}}
     {targetBone {boneName=spine_05}}
     {targetBone {boneName=clavicle_r}}
     {targetBone {boneName=deform_pec_r}}
    ]
    constraintWeights [1,1,1,1,1,1,1,1]
   }
   n20 = "Constraint - master_l_outer_chest_pouch_jnt (PPPPPPPP)"!(
    ComponentPose=@n19
    Alpha=1
    ConstraintWeights_0=0.3508
    ConstraintWeights_1:float=0.1382
    ConstraintWeights_2:float=0.1801
    ConstraintWeights_3:float=0.1511
    ConstraintWeights_4:float=0.0443
    ConstraintWeights_5:float=0.0423
    ConstraintWeights_6:float=0.0744
    ConstraintWeights_7:float=0.0189
   ) -> (Pose) {
    alphaBoolBlend {blendOption=Linear}
    boneToModify {boneName=master_l_outer_chest_pouch_jnt}
    constraintSetup [
     {targetBone {boneName=pelvis}}
     {targetBone {boneName=spine_01}}
     {targetBone {boneName=spine_02}}
     {targetBone {boneName=spine_03}}
     {targetBone {boneName=spine_04}}
     {targetBone {boneName=spine_05}}
     {targetBone {boneName=clavicle_r}}
     {targetBone {boneName=deform_pec_r}}
    ]
    constraintWeights [1,1,1,1,1,1,1,1]
   }
   n21 = "Constraint - master_l_middle_hip_pouch (PPPP)"!(
    ComponentPose=@n20
    Alpha=1
    ConstraintWeights_0=0.4947
    ConstraintWeights_1:float=0.1744
    ConstraintWeights_2:float=0.2082
    ConstraintWeights_3:float=0.1226
   ) -> (Pose) {
    alphaBoolBlend {blendOption=Linear}
    boneToModify {boneName=master_l_middle_hip_pouch}
    constraintSetup [{targetBone {boneName=pelvis}}, {targetBone {boneName=spine_01}}, {targetBone {boneName=spine_02}}, {targetBone {boneName=spine_03}}]
    constraintWeights [1,1,1,1]
   }
   n22 = "Constraint - master_l_inner_hip_pouch (PPPPP)"!(
    ComponentPose=@n21
    Alpha=1
    ConstraintWeights_0=0.4399
    ConstraintWeights_1:float=0.1835
    ConstraintWeights_2:float=0.2243
    ConstraintWeights_3:float=0.1367
    ConstraintWeights_4:float=0.0156
   ) -> (Pose) {
    alphaBoolBlend {blendOption=Linear}
    boneToModify {boneName=master_l_inner_hip_pouch}
    constraintSetup [
     {targetBone {boneName=pelvis}}
     {targetBone {boneName=spine_01}}
     {targetBone {boneName=spine_02}}
     {targetBone {boneName=spine_03}}
     {targetBone {boneName=spine_04}}
    ]
    constraintWeights [1,1,1,1,1]
   }
   n23 = "Constraint - master_l_outer_chest_pouch_jnt (PPPPPPP)"(
    ComponentPose=@n22
    Alpha=1
    ConstraintWeights_0=0.0382
    ConstraintWeights_1=0.1264
    ConstraintWeights_2=0.2589
    ConstraintWeights_3=0.2118
    ConstraintWeights_4=0.0288
    ConstraintWeights_5=0.116
    ConstraintWeights_6=0.2198
   ) {
    alphaBoolBlend {blendOption=Linear}
    boneToModify {boneName=master_l_outer_chest_pouch_jnt}
    constraintSetup [
     {targetBone {boneName=spine_01}}
     {targetBone {boneName=spine_02}}
     {targetBone {boneName=spine_03}}
     {targetBone {boneName=spine_04}}
     {targetBone {boneName=spine_05}}
     {targetBone {boneName=clavicle_l}}
     {targetBone {boneName=deform_pec_l}}
    ]
    constraintWeights [1,1,1,1,1,1,1]
   }
   n24 = "Constraint - master_gun_holster_jnt (PPPP)"!(
    ComponentPose=@n23
    Alpha=1
    ConstraintWeights_0=0.0321
    ConstraintWeights_1:float=0.1649
    ConstraintWeights_2:float=0.7891
    ConstraintWeights_3:float=0.0139
   ) -> (Pose) {
    alphaBoolBlend {blendOption=Linear}
    boneToModify {boneName=master_gun_holster_jnt}
    constraintSetup [
     {targetBone {boneName=pelvis}}
     {targetBone {boneName=thigh_r}}
     {targetBone {boneName=thigh_twist_01_r}}
     {targetBone {boneName=deform_groin_r}}
    ]
    constraintWeights [1,1,1,1]
   }
   n25 = "Constraint - master_l_outer_hip_pouch (PPPP)"!(
    ComponentPose=@n24
    Alpha=1
    ConstraintWeights_0=0.5258
    ConstraintWeights_1:float=0.1649
    ConstraintWeights_2:float=0.1969
    ConstraintWeights_3:float=0.1124
   ) -> (Pose) {
    alphaBoolBlend {blendOption=Linear}
    boneToModify {boneName=master_l_outer_hip_pouch}
    constraintSetup [{targetBone {boneName=pelvis}}, {targetBone {boneName=spine_01}}, {targetBone {boneName=spine_02}}, {targetBone {boneName=spine_03}}]
    constraintWeights [1,1,1,1]
   }
   Save cached pose 'StartPose'(Pose=Component To Local(ComponentPose=@n25)) {cacheName=StartPose}
  ]}
 EventGraph {nodes [
   TryGetPawnOwner()
   Event BlueprintUpdateAnimation()
  ]}
]
variables []
