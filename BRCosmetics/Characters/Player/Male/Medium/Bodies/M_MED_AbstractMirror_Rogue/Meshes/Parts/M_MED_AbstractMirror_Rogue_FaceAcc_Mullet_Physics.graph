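# Physics asset for the mullet hair part of the male medium AbstractMirror Rogue cosmetic body, providing capsule bodies for secondary motion.
# Capsules on the upper spine, neck, head, clavicles and upper arms approximate the body around the hair, linked by free angular constraints.
# Left, center and right chains each hang a tiny master body from the head, constrained to a dynamic mullet body with a very narrow twist limit.
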PhysicsAsset /BRCosmetics/Characters/Player/Male/Medium/Bodies/M_MED_AbstractMirror_Rogue/Meshes/Parts/M_MED_AbstractMirror_Rogue_FaceAcc_Mullet_Physics
bodies {
 clavicle_l [{shapeType=Capsule, length=29, radius=7, shapeName=clavicle_l_capsule, center=<14.9535 1.0882 9.8013>, rotation=<rot 80 -179.9984 -179.9985>}]
 clavicle_r [{
   shapeType=Capsule
   length=29
   radius=7
   shapeName=clavicle_r_capsule
   center=<-17.5198 -1.4378 -9.4903>
   rotation=<rot 80 -179.9984 -179.9988>
  }]
 dyn_C_mullet_jnt [{shapeType=Capsule, length=2, radius=4, shapeName=dyn_C_mullet_jnt_capsule, center=<2.8356 0 0>}]
 dyn_L_mullet_jnt [{shapeType=Capsule, length=2, radius=4, shapeName=dyn_L_mullet_jnt_capsule, center=<2.8356 0 0>}]
 dyn_R_mullet_jnt [{shapeType=Capsule, length=2, radius=4, shapeName=dyn_R_mullet_jnt_capsule, center=<2.8356 0 0>}]
 head [{shapeType=Capsule, length=11.5, radius=7, shapeName=head_capsule, center=<1.9314 -1.4634 0>, rotation=<rot -90 180 -150.0003>}]
 master_C_mullet_jnt [{shapeType=Capsule, length=0.505, radius=0.505, shapeName=master_C_mullet_jnt_capsule, rotation=<rot 0 -90 0>}]
 master_L_mullet_jnt [{shapeType=Capsule, length=0.505, radius=0.505, shapeName=master_L_mullet_jnt_capsule, rotation=<rot 0 -90 0>}]
 master_R_mullet_jnt [{shapeType=Capsule, length=0.505, radius=0.505, shapeName=master_R_mullet_jnt_capsule, rotation=<rot 0 -90 0>}]
 neck_01 [{shapeType=Capsule, length=6, radius=4, shapeName=neck_01_capsule, center=<0 -0.8449 0>, rotation=<rot -90 179.9851 -179.9852>}]
 neck_02 [{shapeType=Capsule, length=6, radius=4, shapeName=neck_02_capsule, rotation=<rot -90 153.9103 -153.9105>}]
 spine_04 [{shapeType=Capsule, length=10.5, radius=10, shapeName=spine_04_capsule, center=<4.4625 -2.8648 0>}]
 spine_05 [{shapeType=Capsule, length=29.5, radius=10, shapeName=spine_05_capsule, center=<2.6994 -2.0247 0>}]
 upperarm_l [{shapeType=Capsule, length=10, radius=7, shapeName=upperarm_l_capsule, center=<7.2174 0 -1.2726>, rotation=<rot -80 179.9997 -179.9997>}]
 upperarm_r [{shapeType=Capsule, length=10, radius=7, shapeName=upperarm_r_capsule, center=<-4.931 0 0.8695>, rotation=<rot -80 179.9997 -179.9997>}]
}
constraints [
 {
  bone1Name=spine_05
  bone2Name=spine_04
  swing1LimitDegrees=45
  swing1Motion=Free
  swing2LimitDegrees=45
  swing2Motion=Free
  twistLimitDegrees=45
  twistMotion=Free
 }
 {
  bone1Name=neck_01
  bone2Name=spine_05
  swing1LimitDegrees=45
  swing1Motion=Free
  swing2LimitDegrees=45
  swing2Motion=Free
  twistLimitDegrees=45
  twistMotion=Free
 }
 {
  bone1Name=neck_02
  bone2Name=neck_01
  swing1LimitDegrees=45
  swing1Motion=Free
  swing2LimitDegrees=45
  swing2Motion=Free
  twistLimitDegrees=45
  twistMotion=Free
 }
 {
  bone1Name=head
  bone2Name=neck_02
  swing1LimitDegrees=45
  swing1Motion=Free
  swing2LimitDegrees=45
  swing2Motion=Free
  twistLimitDegrees=45
  twistMotion=Free
 }
 {
  bone1Name=clavicle_l
  bone2Name=spine_05
  swing1LimitDegrees=45
  swing1Motion=Free
  swing2LimitDegrees=45
  swing2Motion=Free
  twistLimitDegrees=45
  twistMotion=Free
 }
 {
  bone1Name=upperarm_l
  bone2Name=clavicle_l
  swing1LimitDegrees=45
  swing1Motion=Free
  swing2LimitDegrees=45
  swing2Motion=Free
  twistLimitDegrees=45
  twistMotion=Free
 }
 {
  bone1Name=clavicle_r
  bone2Name=spine_05
  swing1LimitDegrees=45
  swing1Motion=Free
  swing2LimitDegrees=45
  swing2Motion=Free
  twistLimitDegrees=45
  twistMotion=Free
 }
 {
  bone1Name=upperarm_r
  bone2Name=clavicle_r
  swing1LimitDegrees=45
  swing1Motion=Free
  swing2LimitDegrees=45
  swing2Motion=Free
  twistLimitDegrees=45
  twistMotion=Free
 }
 {bone1Name=dyn_C_mullet_jnt, bone2Name=master_C_mullet_jnt, swing1LimitDegrees=45, swing2LimitDegrees=45, twistLimitDegrees=1}
 {
  bone1Name=master_C_mullet_jnt
  bone2Name=head
  swing1LimitDegrees=45
  swing1Motion=Free
  swing2LimitDegrees=45
  swing2Motion=Free
  twistLimitDegrees=45
  twistMotion=Free
 }
 {
  bone1Name=master_L_mullet_jnt
  bone2Name=head
  swing1LimitDegrees=45
  swing1Motion=Free
  swing2LimitDegrees=45
  swing2Motion=Free
  twistLimitDegrees=45
  twistMotion=Free
 }
 {bone1Name=dyn_L_mullet_jnt, bone2Name=master_L_mullet_jnt, swing1LimitDegrees=45, swing2LimitDegrees=45, twistLimitDegrees=1}
 {
  bone1Name=master_R_mullet_jnt
  bone2Name=head
  swing1LimitDegrees=45
  swing1Motion=Free
  swing2LimitDegrees=45
  swing2Motion=Free
  twistLimitDegrees=45
  twistMotion=Free
 }
 {bone1Name=dyn_R_mullet_jnt, bone2Name=master_R_mullet_jnt, swing1LimitDegrees=45, swing2LimitDegrees=45, twistLimitDegrees=1}
]
