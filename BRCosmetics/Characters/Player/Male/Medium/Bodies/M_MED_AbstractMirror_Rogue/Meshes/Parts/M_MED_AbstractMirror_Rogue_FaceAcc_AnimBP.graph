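# Animation Blueprint for the AbstractMirror Rogue face accessory, animating headband ribbons, hair strands and mullet over the parent mesh's copied pose.
# Noise-driven wind rotation feeds AnimDynamics and RigidBody simulation, with separate front-end and in-game branches.
# In game, equipping the matching AbstractMirror backbling drops the wind and swaps ribbon simulation for bone constraints and light springs, keeping mullet physics.
# Final bone corrections respond to head rotation, crouching and parachute state.
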AnimBlueprint /BRCosmetics/Characters/Player/Male/Medium/Bodies/M_MED_AbstractMirror_Rogue/Meshes/Parts/M_MED_AbstractMirror_Rogue_FaceAcc_AnimBP
target_skeleton=&/M_MED_AbstractMirror_Rogue_FaceAcc_Skeleton
graphs [
 AnimGraph {nodes [
   n1 = "Constraint - dyn_l_headband_1_jnt (PP)"!(
    ComponentPose=Local To Component(LocalPose=Use cached pose 'HeadbandWindPose'())
    Alpha=1
    ConstraintWeights_0=0.5
    ConstraintWeights_1:float=0.5
   ) -> (Pose) {
    alphaBoolBlend {blendOption=Linear}
    boneToModify {boneName=dyn_l_headband_1_jnt}
    constraintSetup [{targetBone {boneName=head}}, {targetBone {boneName=neck_01}}]
    constraintWeights [1,1]
   }
   n2 = "Constraint - dyn_r_headband_1_jnt (PP)"!(ComponentPose=@n1, Alpha=1, ConstraintWeights_0=0.5, ConstraintWeights_1:float=0.5) -> (Pose) {
    alphaBoolBlend {blendOption=Linear}
    boneToModify {boneName=dyn_r_headband_1_jnt}
    constraintSetup [{targetBone {boneName=head}}, {targetBone {boneName=neck_01}}]
    constraintWeights [1,1]
   }
   n3 = "Constraint - dyn_l_headband_2_jnt (PP)"!(ComponentPose=@n2, Alpha=1, ConstraintWeights_0=0.8, ConstraintWeights_1:float=0.2) -> (Pose) {
    alphaBoolBlend {blendOption=Linear}
    boneToModify {boneName=dyn_l_headband_2_jnt}
    constraintSetup [{targetBone {boneName=neck_01}}, {targetBone {boneName=head}}]
    constraintWeights [1,1]
   }
   n4 = "Constraint - dyn_r_headband_2_jnt (PP)"!(ComponentPose=@n3, Alpha=1, ConstraintWeights_0=0.8, ConstraintWeights_1:float=0.2) -> (Pose) {
    alphaBoolBlend {blendOption=Linear}
    boneToModify {boneName=dyn_r_headband_2_jnt}
    constraintSetup [{targetBone {boneName=neck_01}}, {targetBone {boneName=head}}]
    constraintWeights [1,1]
   }
   n5 = Spring controller - Bone: dyn_l_headband_2_jnt!(ComponentPose:Component Space Pose Link Structure=@n4, Alpha:float=0.04) -> (Pose:Component Space Pose Link Structure) {
    is=title AnimGraphNode_SpringBone
    springDamping=3
    springStiffness=150
    alphaBoolBlend {blendOption=Linear}
    springBone {boneName=dyn_l_headband_2_jnt}
   }
   n6 = "Constraint - dyn_l_headband_3_jnt (PP)"!(
    ComponentPose=Spring controller - Bone: dyn_r_headband_2_jnt!(ComponentPose:Component Space Pose Link Structure=@n5, Alpha:float=0.04) -> (Pose:Component Space Pose Link Structure) {is=title AnimGraphNode_SpringBone, springStiffness=100, alphaBoolBlend {blendOption=Linear}, springBone {boneName=dyn_r_headband_2_jnt}}
    Alpha=1
    ConstraintWeights_0=0.6
    ConstraintWeights_1:float=0.4
   ) -> (Pose) {
    alphaBoolBlend {blendOption=Linear}
    boneToModify {boneName=dyn_l_headband_3_jnt}
    constraintSetup [{targetBone {boneName=spine_05}}, {targetBone {boneName=neck_01}}]
    constraintWeights [1,1]
   }
   n7 = "Constraint - dyn_r_headband_3_jnt (PP)"!(ComponentPose=@n6, Alpha=1, ConstraintWeights_0=0.6, ConstraintWeights_1:float=0.4) -> (Pose) {
    alphaBoolBlend {blendOption=Linear}
    boneToModify {boneName=dyn_r_headband_3_jnt}
    constraintSetup [{targetBone {boneName=spine_05}}, {targetBone {boneName=neck_01}}]
    constraintWeights [1,1]
   }
   n8 = Spring controller - Bone: dyn_l_headband_3_jnt!(ComponentPose:Component Space Pose Link Structure=@n7, Alpha:float=0.04) -> (Pose:Component Space Pose Link Structure) {
    is=title AnimGraphNode_SpringBone
    springDamping=3
    springStiffness=150
    alphaBoolBlend {blendOption=Linear}
    springBone {boneName=dyn_l_headband_3_jnt}
   }
   n9 = "Constraint - dyn_l_headband_4_jnt (PP)"!(
    ComponentPose=Spring controller - Bone: dyn_r_headband_3_jnt!(ComponentPose:Component Space Pose Link Structure=@n8, Alpha:float=0.04) -> (Pose:Component Space Pose Link Structure) {is=title AnimGraphNode_SpringBone, springStiffness=100, alphaBoolBlend {blendOption=Linear}, springBone {boneName=dyn_r_headband_3_jnt}}
    Alpha=1
    ConstraintWeights_0=0.8
    ConstraintWeights_1:float=0.2
   ) -> (Pose) {
    alphaBoolBlend {blendOption=Linear}
    boneToModify {boneName=dyn_l_headband_4_jnt}
    constraintSetup [{targetBone {boneName=spine_04}}, {targetBone {boneName=spine_05}}]
    constraintWeights [1,1]
   }
   n10 = "Constraint - dyn_r_headband_4_jnt (PP)"!(ComponentPose=@n9, Alpha=1, ConstraintWeights_0=0.8, ConstraintWeights_1:float=0.2) -> (Pose) {
    alphaBoolBlend {blendOption=Linear}
    boneToModify {boneName=dyn_r_headband_4_jnt}
    constraintSetup [{targetBone {boneName=spine_04}}, {targetBone {boneName=spine_05}}]
    constraintWeights [1,1]
   }
   n11 = "Constraint - dyn_l_headband_5_jnt (PP)"!(ComponentPose=@n10, Alpha=1, ConstraintWeights_0=0.5, ConstraintWeights_1:float=0.5) -> (Pose) {
    alphaBoolBlend {blendOption=Linear}
    boneToModify {boneName=dyn_l_headband_5_jnt}
    constraintSetup [{targetBone {boneName=spine_04}}, {targetBone {boneName=spine_03}}]
    constraintWeights [1,1]
   }
   n12 = "Constraint - dyn_r_headband_5_jnt (PP)"!(ComponentPose=@n11, Alpha=1, ConstraintWeights_0=0.5, ConstraintWeights_1:float=0.5) -> (Pose) {
    alphaBoolBlend {blendOption=Linear}
    boneToModify {boneName=dyn_r_headband_5_jnt}
    constraintSetup [{targetBone {boneName=spine_04}}, {targetBone {boneName=spine_03}}]
    constraintWeights [1,1]
   }
   n13 = "Transform (Modify) Bone - Bone: dyn_l_headband_4_jnt"(ComponentPose=@n12, Alpha=1, Translation="0,1,0", Rotation="0,0,0", Scale="1,1,1") {
    translationMode=BMM_Additive
    translationSpace=BCS_BoneSpace
    alphaBoolBlend {blendOption=Linear}
    boneToModify {boneName=dyn_l_headband_4_jnt}
    translation {y=1}
   }
   n14 = "Transform (Modify) Bone - Bone: dyn_r_headband_4_jnt"(ComponentPose=@n13, Alpha=1, Translation="0,-1,0", Rotation="0,0,0", Scale="1,1,1") {
    translationMode=BMM_Additive
    translationSpace=BCS_BoneSpace
    alphaBoolBlend {blendOption=Linear}
    boneToModify {boneName=dyn_r_headband_4_jnt}
    translation {y=-1}
   }
   n15 = RigidBody!(ComponentPose=@n14, Alpha, OverrideWorldGravity:Vector="0,0,-980", ExternalForce) -> (Pose) {
    bOverrideWorldGravity=true
    overridePhysicsAsset=&/M_MED_AbstractMirror_Rogue_FaceAcc_Mullet_Physics
    simulationSpace=BaseBoneSpace
    alphaBoolBlend {blendOption=Linear}
    baseBoneRef {boneName=head}
    componentAppliedLinearAccClamp=<3000 3000 3000>
    componentLinearAccScale=<0.3 0.25 0.25>
    componentLinearVelScale=<0.3 0.2 0.2>
    overrideWorldGravity {z=-980}
    simSpaceSettings {maxAngularAcceleration=1000, maxAngularVelocity=3, maxLinearAcceleration=250, maxLinearVelocity=500, velocityScaleZ=0, worldAlpha=1}
   }
   n16 = Anim Dynamics - Bone: dyn_hairstrands_jnt(ComponentPose=Local To Component(LocalPose=Use cached pose 'HeadbandWindPose'()), Alpha=1, ExternalForce="0,0,0") {
    angularDampingOverride=1
    angularSpringConstant=25000
    bAngularSpring=true
    bOverrideAngularDamping=true
    bUseGravityOverride=true
    gravityScale=0.35
    simulationSpace=RootRelative
    alphaBoolBlend {blendOption=Linear}
    boundBone {boneName=dyn_hairstrands_jnt}
    componentAppliedLinearAccClamp=<10000 10000 10000>
    componentLinearAccScale=<0.3 0.2 0.2>
    componentLinearVelScale=<0.3 0.2 0.2>
    gravityOverride {z=-500}
    physicsBodyDefinitions [{
      boundBone {boneName=dyn_hairstrands_jnt}
      boxExtents=<25 25 25>
      constraintSetup {angularLimitsMax=<20 10 35>, angularLimitsMin=<-20 -10 -1>, angularTarget=<1 0 0>}
      localJointOffset=<7 0 0>
     }]
    relativeSpaceBone {boneName=head}
   }
   n17 = RigidBody(ComponentPose=@n16, OverrideWorldGravity=Get GravityOverride()) {
    bOverrideWorldGravity=true
    overridePhysicsAsset=&/M_MED_AbstractMirror_Rogue_FaceAcc__Headband_Physics
    simulationSpace=BaseBoneSpace
    alphaBoolBlend {blendOption=Linear}
    baseBoneRef {boneName=neck_01}
    componentAppliedLinearAccClamp=<3000 3000 3000>
    componentLinearAccScale=<0.1 0.1>
    componentLinearVelScale=<0.1 0.1>
    overrideWorldGravity {z=-980}
    simSpaceSettings {
     maxAngularAcceleration=1000
     maxAngularVelocity=3
     maxLinearAcceleration=250
     maxLinearVelocity=500
     velocityScaleZ=0
     worldAlpha=1
     externalLinearDragV=<0.2 0.2>
    }
   }
   n18 = RigidBody(ComponentPose=@n17, OverrideWorldGravity="0,0,-980") {
    bOverrideWorldGravity=true
    overridePhysicsAsset=&M_MED_AbstractMirror_Rogue_FaceAcc_Mullet_Physics
    simulationSpace=BaseBoneSpace
    alphaBoolBlend {blendOption=Linear}
    baseBoneRef {boneName=head}
    componentAppliedLinearAccClamp=<3000 3000 3000>
    componentLinearAccScale=<0.3 0.25 0.25>
    componentLinearVelScale=<0.3 0.2 0.2>
    overrideWorldGravity {z=-980}
    simSpaceSettings {maxAngularAcceleration=1000, maxAngularVelocity=3, maxLinearAcceleration=250, maxLinearVelocity=500, velocityScaleZ=0, worldAlpha=1}
   }
   n19 = Blend Poses by bool(
    BlendPose_0=Component To Local(ComponentPose=@n15)
    BlendPose_1=Component To Local(ComponentPose=@n18)
    bActiveValue=Get AbstractMirror Backbling Collision Alpha() -> (AbstractMirror Backbling Collision Alpha:bool)
   ) {blendType=Linear}
   n20 = Anim Dynamics - Bone: dyn_hairstrands_jnt(ComponentPose=Local To Component(LocalPose=Use cached pose 'HeadbandWindPose'()), Alpha=1, ExternalForce="0,0,0") {
    angularDampingOverride=1
    angularSpringConstant=25000
    bAngularSpring=true
    bOverrideAngularDamping=true
    bUseGravityOverride=true
    gravityScale=0.35
    simulationSpace=RootRelative
    alphaBoolBlend {blendOption=Linear}
    boundBone {boneName=dyn_hairstrands_jnt}
    componentAppliedLinearAccClamp=<10000 10000 10000>
    componentLinearAccScale=<0.3 0.2 0.2>
    componentLinearVelScale=<0.3 0.2 0.2>
    gravityOverride {z=-500}
    physicsBodyDefinitions [{
      boundBone {boneName=dyn_hairstrands_jnt}
      boxExtents=<25 25 25>
      constraintSetup {angularLimitsMax=<20 10 35>, angularLimitsMin=<-20 -10 -1>, angularTarget=<1 0 0>}
      localJointOffset=<7 0 0>
     }]
    relativeSpaceBone {boneName=head}
   }
   n21 = RigidBody(ComponentPose=@n20, OverrideWorldGravity="0,600,-980") {
    bOverrideWorldGravity=true
    overridePhysicsAsset=&M_MED_AbstractMirror_Rogue_FaceAcc__Headband_Physics
    simulationSpace=BaseBoneSpace
    alphaBoolBlend {blendOption=Linear}
    baseBoneRef {boneName=neck_01}
    componentAppliedLinearAccClamp=<3000 3000 3000>
    componentLinearAccScale=<0.1 0.1>
    componentLinearVelScale=<0.1 0.1>
    overrideWorldGravity {z=-980}
    simSpaceSettings {
     maxAngularAcceleration=1000
     maxAngularVelocity=3
     maxLinearAcceleration=250
     maxLinearVelocity=500
     velocityScaleZ=0
     worldAlpha=1
     externalLinearDragV=<0.2 0.2>
    }
   }
   n22 = RigidBody(ComponentPose=@n21, OverrideWorldGravity="0,0,-980") {
    bOverrideWorldGravity=true
    overridePhysicsAsset=&M_MED_AbstractMirror_Rogue_FaceAcc_Mullet_Physics
    simulationSpace=BaseBoneSpace
    alphaBoolBlend {blendOption=Linear}
    baseBoneRef {boneName=head}
    componentAppliedLinearAccClamp=<3000 3000 3000>
    componentLinearAccScale=<0.3 0.25 0.25>
    componentLinearVelScale=<0.3 0.2 0.2>
    overrideWorldGravity {z=-980}
    simSpaceSettings {maxAngularAcceleration=1000, maxAngularVelocity=3, maxLinearAcceleration=250, maxLinearVelocity=500, velocityScaleZ=0, worldAlpha=1}
   }
   n23 = Component To Local(ComponentPose=@n22)
   n24 = Get bIsInFrontEnd()
   n25 = float * float(A=Get Head R Rot() -> (Head R Rot:double), B=Get Head Back Rot() -> (Head Back Rot:double)) {is=title K2Node_PromotableOperator}
   n26 = "Transform (Modify) Bone - Bone: master_R_mullet_jnt"(
    ComponentPose=Local To Component(LocalPose=Blend Poses by bool(BlendPose_0=@n23, BlendPose_1=@n19, bActiveValue=@n24) {blendType=Linear})
    Alpha=@n25
    Translation="-5,0,0"
    Rotation="0,0,0"
    Scale="1,1,1"
   ) {
    rotationMode=BMM_Additive
    rotationSpace=BCS_BoneSpace
    translationMode=BMM_Additive
    translationSpace=BCS_BoneSpace
    alphaBoolBlend {blendOption=Linear}
    boneToModify {boneName=master_R_mullet_jnt}
    translation {x=-5}
   }
   n27 = "Transform (Modify) Bone - Bone: master_C_mullet_jnt"(ComponentPose=@n26, Alpha=@n25, Translation="-5,0,0", Rotation="0,0,0", Scale="1,1,1") {
    rotationMode=BMM_Additive
    rotationSpace=BCS_BoneSpace
    translationMode=BMM_Additive
    translationSpace=BCS_BoneSpace
    alphaBoolBlend {blendOption=Linear}
    boneToModify {boneName=master_C_mullet_jnt}
    translation {x=-5}
   }
   n28 = Get Parachute Deployed() -> (Parachute Deployed:bool)
   n29 = "Constraint - dyn_l_headband_5_jnt (P)"!(ComponentPose=@n27, bAlphaBoolEnabled:bool=@n28, ConstraintWeights_0=1) -> (Pose) {
    alphaInputType=Bool
    alphaBoolBlend {blendOption=Linear}
    boneToModify {boneName=dyn_l_headband_5_jnt}
    constraintSetup [{targetBone {boneName=spine_05}}]
    constraintWeights [1]
   }
   n30 = "Constraint - dyn_r_headband_5_jnt (P)"!(ComponentPose=@n29, bAlphaBoolEnabled:bool=@n28, ConstraintWeights_0=1) -> (Pose) {
    alphaInputType=Bool
    alphaBoolBlend {blendOption=Linear}
    boneToModify {boneName=dyn_r_headband_5_jnt}
    constraintSetup [{targetBone {boneName=spine_05}}]
    constraintWeights [1]
   }
   n31 = "Transform (Modify) Bone - Bone: master_headband_ribbons_jnt"(
    ComponentPose=@n30
    Alpha=float * float!(A=Get Crouching() -> (Crouching:double), B=Get Head Back Rot(), C:double=Get Not Cosmetic Set Float() -> (Not Cosmetic Set Float:double)) -> (ReturnValue)
    Translation="0,0,0"
    Rotation="0.000,-0.000,20.000"
    Scale="1,1,1"
   ) {
    rotationMode=BMM_Additive
    rotationSpace=BCS_BoneSpace
    alphaBoolBlend {blendOption=Linear}
    boneToModify {boneName=master_headband_ribbons_jnt}
    rotation {roll=20}
   }
   n32 = "Transform (Modify) Bone - Bone: master_headband_ribbons_jnt"(
    ComponentPose=@n31
    Alpha=float * float(A=Get Crouching(), B=Get Head R Rot(), C=Get Not Cosmetic Set Float())
    Translation="0,0,0"
    Rotation="0.000,-0.000,20.000"
    Scale="1,1,1"
   ) {
    rotationMode=BMM_Additive
    rotationSpace=BCS_BoneSpace
    alphaBoolBlend {blendOption=Linear}
    boneToModify {boneName=master_headband_ribbons_jnt}
    rotation {roll=20}
   }
   Output Pose(Result=Component To Local(ComponentPose=@n32)) {layerGroup=None}
   Save cached pose 'StartPose'(Pose=Copy Pose From Mesh() {bUseAttachedParent=true}) {cacheName=StartPose}
   n33 = AND Boolean(A=NOT Boolean(A=Get Parachute Deployed()), B=Get bRigidBodyEnabled())
   n34 = "Transform (Modify) Bone - Bone: master_l_headband_jnt"!(
    ComponentPose=Local To Component(LocalPose=Use cached pose 'StartPose'())
    bAlphaBoolEnabled:bool=@n33
    Translation="0,0,0"
    Rotation=Get HairRotator_01() -> (HairRotator_01:Rotator)
    Scale="1,1,1"
   ) -> (Pose) {
    alphaInputType=Bool
    rotationMode=BMM_Additive
    rotationSpace=BCS_BoneSpace
    alphaBoolBlend {blendOption=Linear}
    boneToModify {boneName=master_l_headband_jnt}
   }
   n35 = "Transform (Modify) Bone - Bone: master_r_headband_jnt"!(
    ComponentPose=@n34
    bAlphaBoolEnabled:bool=@n33
    Translation="0,0,0"
    Rotation=Get HairRotator_02() -> (HairRotator_02:Rotator)
    Scale="1,1,1"
   ) -> (Pose) {
    alphaInputType=Bool
    rotationMode=BMM_Additive
    rotationSpace=BCS_BoneSpace
    alphaBoolBlend {blendOption=Linear}
    boneToModify {boneName=master_r_headband_jnt}
   }
   n36 = Component To Local(ComponentPose=@n35)
   n37 = Use cached pose 'StartPose'()
   n38 = Get AbstractMirror Backbling Collision Alpha()
   n39 = "Transform (Modify) Bone - Bone: master_l_headband_jnt"!(
    ComponentPose=Local To Component(LocalPose=Use cached pose 'StartPose'())
    Alpha:float=0.25
    Translation="0,0,0"
    Rotation=Get HairRotator_01()
    Scale="1,1,1"
   ) -> (Pose) {rotationMode=BMM_Additive, rotationSpace=BCS_BoneSpace, alphaBoolBlend {blendOption=Linear}, boneToModify {boneName=master_l_headband_jnt}}
   n40 = "Transform (Modify) Bone - Bone: master_r_headband_jnt"!(ComponentPose=@n39, Alpha:float=0.25, Translation="0,0,0", Rotation=Get HairRotator_02(), Scale="1,1,1") -> (Pose) {rotationMode=BMM_Additive, rotationSpace=BCS_BoneSpace, alphaBoolBlend {blendOption=Linear}, boneToModify {boneName=master_r_headband_jnt}}
   Save cached pose 'HeadbandWindPose'(Pose=Blend Poses by bool(
     BlendPose_0=Component To Local(ComponentPose=@n40)
     BlendPose_1=Blend Poses by bool(BlendPose_0=@n37, BlendPose_1=@n36, bActiveValue=@n38) {blendType=Linear}
     bActiveValue=Get bIsInFrontEnd()
    ) {blendType=Linear}) {cacheName=HeadbandWindPose}
   Get GravityOverride()
  ]}
 EventGraph {nodes [
   n1 = Event BlueprintUpdateAnimation()
   > n2 = Sequence!(execute) -> (then_0, then_1, then_2:Exec, then_3:Exec)
   > Set Delta Time(Delta Time:double=float + float(A=@n1.DeltaTimeX, B=Get Delta Time() -> (Delta Time:double))) -> (Output_Get:double)
   n3 = Get Delta Time()
   n4 = Gravity Override State Function DEPRECATED(execute=@n2.then_1, CharacterPartAnimInstance=Self-Reference(), GameTimeInSeconds=@n3, GravityOverrideParams=Get Gravity Override Params())
   > Set GravityOverride(GravityOverride=@n4)
   n5 = GetOwnerSkeletalMeshComponent()
   n6 = BreakRotator(InRot=BreakTransform(InTransform=GetDeltaTransformFromRefPose(self=@n5, BoneName=head, BaseName=spine_05)).Rotation)
   > Set Head Back Rot(Head Back Rot:double=MapRangeClamped(Value=@n6.Yaw, InRangeA=-10, InRangeB=-55)) -> (Output_Get:double)
   > Set Head R Rot(Head R Rot:double=MapRangeClamped(Value=@n6.Pitch, InRangeA=0, InRangeB=45)) -> (Output_Get:double)
   > Set Parachute Deployed(Parachute Deployed:bool=Get bIsParachuteOpened()) -> (Output_Get:bool)
   > Set Crouching(Crouching:double="To Float (Boolean)"(InBool=Get bIsCrouching())) -> (Output_Get:double)
   n7 = RandomRotatorFunction!(
    execute=@n2.then_2
    GameTimeInSeconds:double=@n3
    PerlinScale:double=1
    Frequency:Rotator="5.5,5,0"
    Amplitude:Rotator="25,25,0"
    Offset:Rotator="0,-25,0"
    TimeOffset:double=0
   ) -> (then, RandomRotator:Rotator)
   > Set HairRotator_01(HairRotator_01:Rotator=@n7) -> (Output_Get:Rotator)
   n8 = RandomRotatorFunction(execute=@n2.then_3, GameTimeInSeconds=@n3, PerlinScale=1, Frequency="5.5,5,0", Amplitude="-25,-25,0", Offset="0,-35,0", TimeOffset=0)
   > Set HairRotator_02(HairRotator_02:Rotator=@n8) -> (Output_Get:Rotator)
   Event OnCharacterCustomizationLoaded()
   > Get Dynamic Material Instance(SkeletalMeshComponent=GetPartSkeletalMeshComponent())
   > Sequence!(execute) -> (then_0, then_1)
   > n9 = Branch(Condition=BackpackItemDefinitionHasAnyMetaTag!(Tags:Array of Gameplay Tag Structures=Make Array("[0]"=(TagName="Cosmetics.Metatags.AbstractMirror")), bRequiresEquipped:bool=false) -> (ReturnValue:bool))
   > n10 = Set AbstractMirror Backbling Collision Alpha(AbstractMirror Backbling Collision Alpha:bool=true) -> (Output_Get:bool)
   > Set Not Cosmetic Set Float(Not Cosmetic Set Float:double="To Float (Boolean)"(InBool=NOT Boolean(A=@n10))) -> (Output_Get:double)
   n11 = Set AbstractMirror Backbling Collision Alpha(execute=@n9.else, AbstractMirror Backbling Collision Alpha=false)
   > Set Not Cosmetic Set Float(Not Cosmetic Set Float="To Float (Boolean)"(InBool=NOT Boolean(A=@n11)))
   TryGetPawnOwner()
  ]}
]
variables [
 AbstractMirror Backbling Collision Alpha {type=boolean}
 Crouching {type=number}
 Delta Time {type=number}
 Gravity Override Params {
  type=GravityOverrideParamsStruct
  default {
   onGround_Moving {gravityOverride=<0 -50 -15>, windAmplitude=<50 25 50>, windFrequency=<0.1 0.2 0.15>, windOffset=<0 -50 -50>}
   onGround_Standing {gravityOverride=<0 -50 -15>, windAmplitude=<50 25 50>, windFrequency=<0.1 0.2 0.15>, windOffset=<0 -50 -50>}
  }
 }
 GravityOverride {type=Vector}
 HairRotator_01 {type=Rotator}
 HairRotator_02 {type=Rotator}
 Head Back Rot {type=number}
 Head R Rot {type=number}
 Not Cosmetic Set Float {type=number}
 Parachute Deployed {type=boolean}
]
